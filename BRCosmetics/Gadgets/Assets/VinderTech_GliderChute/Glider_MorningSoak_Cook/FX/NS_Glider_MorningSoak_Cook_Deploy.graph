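# Niagara deploy effect for the MorningSoak Cook glider: a short-lived burst of sprite smoke.
# A single Smoke emitter spawns instantaneously, tints particles by System.TOD, pushes them with a camera offset and keeps them on a socket read through System.SocketReader.
# Over life, a curve grows the sprite from zero while another ramps a dynamic material parameter.
# Front-end modifiers wrap size and socket offset but use identical values to gameplay.
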NiagaraSystem /BRCosmetics/Gadgets/Assets/VinderTech_GliderChute/Glider_MorningSoak_Cook/FX/NS_Glider_MorningSoak_Cook_Deploy
dependencies {
 usedDataInterfaces [&NiagaraDataInterfaceCurve, &NiagaraDataInterfaceSocketReader, &NiagaraDataInterfaceColorCurve]
 usedDynamicInputs [&MakeFloatFromLinearColor, &ColorFromCurve]
 usedModules [
  &self:SystemSpawnScript.SystemScriptSource.SystemScriptGraph.NiagaraNodeAssignment_0.SetVariables_#1
  &self:SystemSpawnScript.SystemScriptSource.SystemScriptGraph.NiagaraNodeAssignment_1.SetVariables_#2
  &SystemState
 ]
}
emitters {Smoke {
  emitterUpdateScript [
   &EmitterState()
   &SpawnBurst_Instantaneous(Use Spawn Probability=false, UseLoopCountLimit=false)
  ]
  particleSpawnScript [
   &InitializeParticle(
    Color=&MultiplyLinearColor(
     Linear Color=<rgba 1.1 1.1 1.1 1>
     Scale Factor=&MakeVector4_FromLinearColor(LinearColor=/Niagara/DynamicInputs/TypeConversions/MakeLinearColorFromFloat(Value:NiagaraFloat=$System.TOD))
    )
    Lifetime=0.35
    Sprite Rotation Mode=::NewEnumerator2
    Sprite Size Mode=::NewEnumerator3
    Uniform Sprite Size=&Multiply_Float(A=400, B=&FN_FrontEndFloatModifier(Gameplay=1, Front End=1))
    UsePositionOffset=false
    Write Material Random=true
   )
   &CameraOffset(Camera Offset Amount=90)
  ]
  particleUpdateScript [
   &ParticleState()
   &SocketLocation(
    SamplePosOffset=true
    Sampled Position Offset=&FN_FrontEndVector3Modifier(Gameplay=<0 0 140>, Front End=<0 0 140>)
    Socket Index=1
    Socket Reader=$System.SocketReader
    Socket Sampling Mode=::NewEnumerator3
    VelocitySpeedLimit=false
   )
   &ScaleSpriteSize(Uniform Curve Sprite Scale=DataInterface() {propertyValues {Curve {
       defaultValue=0
       keys [
        {arriveTangent=2.0491, interpMode=RCIM_Cubic, leaveTangent=2.0491, tangentMode=RCTM_User, time=0, value=0}
        {interpMode=RCIM_Cubic, tangentMode=RCTM_User, time=1, value=1}
       ]
      }}})
   &DynamicMaterialParameters(
    Index 0 Param 1=&FloatFromCurve(FloatCurve=DataInterface() {propertyValues {Curve {keys [<at 0 0>,<at 1 1>]}}})
    Module.Index0Float0Write=true
    Module.Index0Float1Write=true
    Module.Index0Float2Write=true
    Module.Index0Float3Write=true
   )
  ]
  rendererClasses [&NiagaraSpriteRendererProperties]
  renderers [{rendererClass=&NiagaraSpriteRendererProperties}]
 }}
summary {
 systemName=NS_Glider_MorningSoak_Cook_Deploy
 emitters [{emitterName=Smoke, rendererClasses [&NiagaraSpriteRendererProperties]}]
 userVariables [
  User.bIsFrontEnd {type=NiagaraBool}
  User.bIsFrontEndPreview {type=NiagaraBool}
 ]
}
