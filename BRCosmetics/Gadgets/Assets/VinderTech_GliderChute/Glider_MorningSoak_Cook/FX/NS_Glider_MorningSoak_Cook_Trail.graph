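# Niagara trail effect for the MorningSoak Cook glider, combining a socket-attached fire, a fake light with fan, ribbon wing trails, a front wind mesh and speed lines.
# Fire and FakeLight_And_Fan each serve two sprite/mesh elements, using the spawn group as VisibilityTag to switch sockets, motion and color.
# System EffectIntensity and Timeline scale the effects, speed lines follow ForwardDot, and front-end, time-of-day and platform switches adjust the look.
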NiagaraSystem /BRCosmetics/Gadgets/Assets/VinderTech_GliderChute/Glider_MorningSoak_Cook/FX/NS_Glider_MorningSoak_Cook_Trail
dependencies {
 usedDataInterfaces [
  &NiagaraDataInterfaceCurve
  &NiagaraDataInterfaceArrayFloat
  &NiagaraDataInterfacePlatformSet
  &NiagaraDataInterfaceVector2DCurve
  &NiagaraDataInterfaceColorCurve
  &NiagaraDataInterfaceVectorCurve
  &NiagaraDataInterfaceSocketReader
 ]
 usedDynamicInputs [&Boolean_AndOperation]
 usedModules [
  &FN_VariantManager
  &self:SystemSpawnScript.SystemScriptSource.SystemScriptGraph.NiagaraNodeAssignment_1.SetVariables_#1
  &SystemState
  &FN_Integrator
  &Timeline
  &self:SystemSpawnScript.SystemScriptSource.SystemScriptGraph.NiagaraNodeAssignment_2.SetVariables_#2
 ]
}
emitters {
 FakeLight_And_Fan {
  emitterUpdateScript [
   &EmitterState(Life Cycle Mode=::NewEnumerator1, Loop Duration=1, UseLoopDelay=false)
   &SpawnBurst_Instantaneous(Use Spawn Probability=false, UseLoopCountLimit=false) {module=SpawnBurst_Instantaneous001}
   &SpawnBurst_Instantaneous(Spawn Group=1, Use Spawn Probability=false, UseLoopCountLimit=false)
  ]
  particleSpawnScript [
   /SetVariables(Particles.VisibilityTag=$Emitter.SpawnGroup) {isSetParameters=true}
   &InitializeParticle(
    Mesh Scale Mode=::NewEnumerator3
    Mesh Uniform Scale=0.105
    Sprite Size Mode=::NewEnumerator3
    Uniform Sprite Size=80
    UsePositionOffset=false
    Write Material Random=true
   )
  ]
  particleUpdateScript [
   &ParticleState()
   &SocketLocation(
    SamplePosOffset=true
    Sampled Position Offset=&MergeHelper_VectorSwitch(
     0=&FN_FrontEndVector3Modifier(Gameplay=<0 0 35>, Front End=<0 0 0>)
     1=<-3 -1 -48>
     2=?
     3=?
     Index=$Particles.VisibilityTag
     Num=::NewEnumerator0
    )
    Socket Index=&MergeHelper_IntSwitch(0=7, 1=6, 2=?, 3=?, 4=?, 5=?, Index=$Particles.VisibilityTag, Num=::NewEnumerator0)
    Socket Reader=$System.Socket Reader
    Socket Sampling Mode=::NewEnumerator3
    VelocitySpeedLimit=false
   ) {module=SocketLocation001}
   &SolveForcesAndVelocity(Clamp Velocity=false, Limit Acceleration=false)
   &ScaleColor(
    Scale Alpha=&MergeHelper_FloatSwitch(
     0=&FloatFromCurve(FloatCurve=DataInterface() {propertyValues {Curve {
         defaultValue=0.3537
         keys [
          {interpMode=RCIM_Cubic, tangentMode=RCTM_User, time=0, value=0.7}
          {interpMode=RCIM_Cubic, tangentMode=RCTM_User, time=0.1399, value=1}
          {arriveTangent=2.4787, interpMode=RCIM_Cubic, leaveTangent=2.4787, tangentMode=RCTM_User, time=0.3809, value=0.6775}
          {arriveTangent=-2.5809, interpMode=RCIM_Cubic, leaveTangent=-2.5809, tangentMode=RCTM_User, time=0.785, value=0.6377}
          {tangentMode=RCTM_User, time=1, value=0.7}
         ]
        }}})
     1=1
     2=?
     3=?
     Index=$Particles.VisibilityTag
     Num=::NewEnumerator0
    )
    Scale RGB=&MergeHelper_VectorSwitch(
     0=&Multiply_Vector(
      A=&MakeVectorFromLinearColorRGB(Color=<rgba 0.6042 0.1656 0.0063 1>)
      B=&Multiply_Vector(
       A=&VectorFromFloat(Value=&FN_PickFloatDependOnPlatform(TargetPlatform=DataInterface() {propertyValues {Platforms {qualityLevelMask=-29}}}, TargetPlatformActive=0.1, TargetPlatformNoActive=1))
       B=&VectorFromFloat(Value=&Lerp_Float(A=1, B=3, Alpha=$System.EffectIntensity))
      )
     )
     1=<0 0 0>
     2=?
     3=?
     Index=$Particles.VisibilityTag
     Num=::NewEnumerator0
    )
    ScaleA=true
    ScaleRGB=true
   )
   &DynamicMaterialParameters(
    Index 0 Param 1=&Lerp_Float(A=0.2, B=0.5, Alpha=$System.Timeline.Time)
    Index 0 Param 2=1
    Module.Param0WriteEnabled=true
    Module.Param1WriteEnabled=true
    Module.Param2WriteEnabled=false
    Module.Param3WriteEnabled=true
   )
  ]
  rendererClasses [&NiagaraSpriteRendererProperties, &NiagaraMeshRendererProperties]
  renderers [
   {rendererClass=&NiagaraSpriteRendererProperties}
   {rendererClass=&NiagaraMeshRendererProperties, rendererIndex=1}
  ]
 }
 Fire {
  emitterUpdateScript [
   &EmitterState()
   &SpawnRate(
    SpawnGroup=1
    SpawnRate=&FN_FrontEndFloatModifier(Gameplay=&Lerp_Float(A=0, B=25, Alpha=$System.EffectIntensity), Front End=25)
    Use Spawn Probability=false
   )
   &SpawnRate(
    SpawnRate=&FN_FrontEndFloatModifier(Gameplay=&Lerp_Float(A=20, B=40, Alpha=$System.EffectIntensity), Front End=40)
    Use Spawn Probability=false
   ) {module=SpawnRate002}
  ]
  particleSpawnScript [
   /SetVariables(Particles.VisibilityTag=$Emitter.SpawnGroup) {isSetParameters=true}
   &InitializeParticle(
    Lifetime=&MergeHelper_FloatSwitch(
     0=&RandomRangeFloat(Minimum=0.25, Maximum=0.45)
     1=&RandomRangeFloat(Minimum=0.35, Maximum=0.5)
     2=?
     3=?
     Index=$Emitter.SpawnGroup
     Num=::NewEnumerator0
    )
    Mesh Scale Mode=::NewEnumerator3
    Mesh Uniform Scale=&FN_FrontEndFloatModifier(Gameplay=1, Front End=0.6)
    Position Offset=<0 0 150>
    Sprite Size Mode=::NewEnumerator3
    Sprite UV Mode=::NewEnumerator1
    Uniform Sprite Size=7
    UsePositionOffset=true
    Write Material Random=true
   )
   &AddVelocity(
    Scale Added Velocity=&Multiply_VectorByFloat(
     Vector=&Multiply_VectorByFloat(Vector=<2.5 2.5 2.5>, Float=1)
     Float=&Lerp_Float(A=1, B=0.6, Alpha=0.5)
    )
    Velocity=&MergeHelper_VectorSwitch(0=<0 0 0>, 1=<0 0 50>, 2=?, 3=?, Index=$Particles.VisibilityTag, Num=::NewEnumerator0)
   )
   &AddVelocity(
    Cone Angle=&Lerp_Float(A=0, B=17, Alpha=$System.EffectIntensity)
    Cone Axis=&Lerp_Vector(A=<0 0 1>, B=<0 1 0.8>, Alpha=$System.EffectIntensity)
    Module.UseConeFalloff=false
    Module.UseVelDistribution=false
    Override Seed=false
    Velocity Mode=::NewEnumerator2
    Velocity Speed=&MergeHelper_FloatSwitch(
     0=&Multiply_Float(
      A=&RandomRangeFloat(Minimum=0.5, Maximum=1)
      B=&Lerp_Float(
       A=80
       B=&Multiply_Float(
        A=370
        B=&UniformAOrBFloat(A=1, B=1.333, Distribution Weight=0.1, Evaluation Type=::NewEnumerator0, Recalculate Random Each Loop=true, Override Seed=false)
       )
       Alpha=$System.EffectIntensity
      )
     )
     1=0
     2=?
     3=?
     Index=$Particles.VisibilityTag
     Num=::NewEnumerator0
    )
   ) {module=AddVelocity001}
   &AddVelocityInCone(
    Cone Angle=50
    Cone Axis=<0 0 1>
    Module.Use Velocity Falloff On Cone Axis=false
    Velocity Strength=&MergeHelper_FloatSwitch(0=0, 1=70, 2=?, 3=?, Index=$Particles.VisibilityTag, Num=::NewEnumerator0)
   ) {module=AddVelocityInCone001}
   /SetVariables(Particles.SmokeCameraOffset=-10) {isSetParameters=true}
  ]
  particleUpdateScript [
   &ParticleState()
   &SocketLocation(
    SamplePosOffset=true
    Sampled Position Offset=&MergeHelper_VectorSwitch(0=<0 0 5>, 1=<0 0 15>, 2=?, 3=?, Index=$Particles.VisibilityTag, Num=::NewEnumerator0)
    Sampled Position Offset Coordinate Space:ENiagaraSkelMeshTransforms=::NewEnumerator2
    Socket Index=&MergeHelper_IntSwitch(0=5, 1=5, 2=?, 3=?, 4=?, 5=?, Index=$Particles.VisibilityTag, Num=::NewEnumerator0)
    Socket Reader=$System.Socket Reader
    Socket Sampling Mode=::NewEnumerator3
    VelocitySpeedLimit=false
   )
   &ScaleMeshSize(Scale Factor=&Multiply_VectorByFloat(
     Vector=&Multiply_VectorByFloat(
      Vector=&Multiply_Vector(
       A=<1 1 1>
       B=&VectorFromCurve(VectorCurve=DataInterface() {propertyValues {XCurve {keys [<at 0 5>,<at 0.2 1>,<at 1 1>]}, YCurve {keys [<at 0 1>,<at 1 1>]}, ZCurve {keys [<at 0 1>,<at 1 1>]}}})
      )
      Float=&RandomRangeFloat(Minimum=1, Maximum=1.5)
     )
     Float=&FloatFromCurve(
      FloatCurve=DataInterface() {propertyValues {Curve {keys [
          {interpMode=RCIM_Cubic, tangentMode=RCTM_User, time=0, value=0}
          {arriveTangent=1.3433, tangentMode=RCTM_User, time=0.2, value=0.35}
          {arriveTangent=1.4526, interpMode=RCIM_Cubic, leaveTangent=1.4526, tangentMode=RCTM_User, time=1, value=1.5}
         ]}}}
      Scale Curve=&Multiply_Float(A=0.83, B=1)
     )
    ))
   &ScaleSpriteSize(
    Scale Factor=&Multiply_Vector2D(A=&RandomRangeVector2D(Minimum=<2 2>, Maximum=<3 3>), B=<1 1>)
    Scale Sprite Size Mode=::NewEnumerator2
   )
   &ScaleSpriteSize(
    Non-Uniform Curve Scale=<1 1.5>
    Non-Uniform Curve Sprite Scale=DataInterface() {propertyValues {
      XCurve {keys [
        {arriveTangent=0.8609, interpMode=RCIM_Cubic, leaveTangent=0.8609, tangentMode=RCTM_User, time=0, value=0.6}
        {arriveTangent=0.5103, interpMode=RCIM_Cubic, leaveTangent=0.5103, tangentMode=RCTM_User, time=0.3, value=0.8}
        {arriveTangent=0.0224, interpMode=RCIM_Cubic, leaveTangent=0.0224, tangentMode=RCTM_User, time=1, value=1}
       ]}
      YCurve {keys [
        {arriveTangent=0.8801, interpMode=RCIM_Cubic, leaveTangent=0.8801, tangentMode=RCTM_User, time=0, value=0.3}
        {arriveTangent=0.6586, interpMode=RCIM_Cubic, leaveTangent=0.6586, tangentMode=RCTM_User, time=0.3, value=0.6}
        {arriveTangent=0.0224, interpMode=RCIM_Cubic, leaveTangent=0.0224, tangentMode=RCTM_User, time=1, value=1}
       ]}
     }}
    Scale Sprite Size Mode=::NewEnumerator3
   ) {module=ScaleSpriteSize001}
   &ScaleSpriteSize(
    Scale Factor=&MakeVector2D(
     X=&Lerp_Float(
      A=1
      B=&FloatFromCurve(
       FloatCurve=DataInterface() {propertyValues {Curve {keys [
           {interpMode=RCIM_Cubic, leaveTangent=4.962, tangentMode=RCTM_Break, time=0.2, value=0.8}
           {arriveTangent=0.4581, interpMode=RCIM_Cubic, tangentMode=RCTM_Break, time=1, value=3}
          ]}}}
       Scale Curve=1.2
      )
      Alpha=$System.EffectIntensity
     )
     Y=&Lerp_Float(
      A=1
      B=&FloatFromCurve(
       FloatCurve=DataInterface() {propertyValues {Curve {keys [
           {interpMode=RCIM_Cubic, leaveTangent=1.2424, tangentMode=RCTM_Break, time=0.2, value=1}
           {arriveTangent=0.0595, interpMode=RCIM_Cubic, tangentMode=RCTM_Break, time=1, value=1.5}
          ]}}}
       Scale Curve=1.2
      )
      Alpha=$System.EffectIntensity
     )
    )
    Scale Sprite Size Mode=::NewEnumerator2
   ) {module=ScaleSpriteSize003}
   &Drag(
    Drag=&MergeHelper_FloatSwitch(0=0, 1=1, 2=?, 3=?, Index=$Particles.VisibilityTag, Num=::NewEnumerator0)
    Use Linear Drag=true
    Use Rotational Drag=true
   )
   &SolveForcesAndVelocity(Clamp Velocity=false, Limit Acceleration=false)
   &ScaleSpriteSizeBySpeed(Max Scale Factor=<1 4>)
   /SetVariables(Particles.SmokeDynamicMaterialParameter:Vector4f=&MakeVector4(
     X=&FloatFromCurve(
      FloatCurve=DataInterface() {propertyValues {Curve {keys [
          {interpMode=RCIM_Cubic, tangentMode=RCTM_User, time=0, value=0}
          <at 0.2 0.55>
          {arriveTangent=-1.0223, interpMode=RCIM_Cubic, leaveTangent=-1.0223, tangentMode=RCTM_User, time=1, value=1}
         ]}}}
      Scale Curve=&RandomRangeFloat(Minimum=0.7, Maximum=1.5)
     )
     Y=0
     Z=0
     W=&Lerp_Float(
      A=0.6
      B=0
      Alpha=&FloatFromCurve(FloatCurve=DataInterface() {propertyValues {Curve {keys [
           {arriveTangent=0.7438, interpMode=RCIM_Cubic, leaveTangent=0.7438, tangentMode=RCTM_User, time=0, value=0}
           {arriveTangent=0.0023, interpMode=RCIM_Cubic, leaveTangent=0.0023, tangentMode=RCTM_User, time=1, value=1}
          ]}}})
     )
    )) {isSetParameters=true}
   &DynamicMaterialParameters(
    Index 0 Param 1=&FloatFromCurve(FloatCurve=DataInterface() {propertyValues {Curve {keys [
         {arriveTangent=-0.1846, time=0, value=0}
         {arriveTangent=1.3385, interpMode=RCIM_Cubic, leaveTangent=1.1594, tangentMode=RCTM_Break, time=0.5, value=0}
         {arriveTangent=2, interpMode=RCIM_Cubic, leaveTangent=2, time=0.9, value=0.45}
         {arriveTangent=3.0091, interpMode=RCIM_Cubic, leaveTangent=3.0091, tangentMode=RCTM_User, time=1, value=1}
        ]}}})
    Index 0 Param 2=1
    Index 0 Param 3=&Lerp_Float(A=0, B=1, Alpha=&Multiply_Float(A=$System.EffectIntensity, B=1))
    Index 0 Param 4=&Lerp_Float(A=0, B=1, Alpha=&Multiply_Float(A=$System.EffectIntensity, B=1))
    Index 1 Param 1=&Lerp_Float(A=1, B=0, Alpha=$System.EffectIntensity)
    Module.Index0Param0Write=true
    Module.Index0Param1Write=true
    Module.Index0Param2Write=true
    Module.Index0Param3Write=true
    Module.Index1Param0Write=false
    Module.Index1Param1Write=false
    Module.Index1Param2Write=false
    Module.Index1Param3Write=false
    Write Parameter Index 1=true
   )
   &ScaleColor(
    Scale Alpha=&MergeHelper_FloatSwitch(
     0=&Multiply_Float(
      A=&FloatFromCurve(
       FloatCurve=DataInterface() {propertyValues {Curve {keys [
           {arriveTangent=9.0725, interpMode=RCIM_Cubic, leaveTangent=9.0725, tangentMode=RCTM_User, time=0, value=1}
           {arriveTangent=-2.5249, interpMode=RCIM_Cubic, leaveTangent=-2.5249, tangentMode=RCTM_User, time=0.1408, value=0.9847}
           {arriveTangent=-1.8607, interpMode=RCIM_Cubic, leaveTangent=-1.8607, tangentMode=RCTM_User, time=0.3012, value=0.9413}
           {arriveTangent=-1.0655, interpMode=RCIM_Cubic, leaveTangent=-1.0655, tangentMode=RCTM_User, time=0.5976, value=0.5305}
           {arriveTangent=-0.0405, interpMode=RCIM_Cubic, leaveTangent=-0.0405, tangentMode=RCTM_User, time=1, value=0}
          ]}}}
       Scale Curve=2
      )
      B=&Multiply_Float(A=&Lerp_Float(A=2, B=2, Alpha=$System.EffectIntensity), B=1)
     )
     1=&FloatFromCurve(
      FloatCurve=DataInterface() {propertyValues {Curve {
         defaultValue=0
         keys [
          {arriveTangent=9.0725, interpMode=RCIM_Cubic, leaveTangent=9.0725, tangentMode=RCTM_User, time=0, value=0}
          {arriveTangent=-2.5249, interpMode=RCIM_Cubic, leaveTangent=-2.5249, tangentMode=RCTM_User, time=0.1408, value=0.9847}
          {arriveTangent=-0.0405, interpMode=RCIM_Cubic, leaveTangent=-0.0405, tangentMode=RCTM_User, time=1, value=0}
         ]
        }}}
      Scale Curve=&FN_FrontEndFloatModifier(
       Gameplay=&Multiply_Float(A=&Multiply_Float(A=1.55, B=$System.Timeline.Time), B=1)
       Front End=1
      )
     )
     2=?
     3=?
     Index=$Particles.VisibilityTag
     Num=::NewEnumerator0
    )
    Scale RGB=&MergeHelper_VectorSwitch(
     0=&Multiply_VectorByFloat(
      Vector=&MakeVectorFromLinearColorRGB(Color=&ColorFromCurve(ColorCurve=DataInterface() {propertyValues {
          AlphaCurve {keys [<at 0 1 RCIM_Cubic>,<at 1 1 RCIM_Cubic>]}
          BlueCurve {keys [
            <at 0.034 0>
            <at 0.1599 0>
            <at 0.3037 0 RCIM_Cubic>
            {arriveTangent=0.1755, interpMode=RCIM_Cubic, leaveTangent=0.1755, time=0.4429, value=0}
            <at 0.5885 0.05>
           ]}
          GreenCurve {keys [
            <at 0.034 0.5833>
            <at 0.1599 0.4793>
            {arriveTangent=-1.5634, interpMode=RCIM_Cubic, leaveTangent=-1.5634, time=0.3037, value=0.2667}
            {arriveTangent=-0.7285, interpMode=RCIM_Cubic, leaveTangent=-0.7285, time=0.4429, value=0.037}
            <at 0.5885 0.0592>
           ]}
          RedCurve {keys [
            <at 0.034 1>
            <at 0.1599 1>
            {arriveTangent=-1.9652, interpMode=RCIM_Cubic, leaveTangent=-1.9652, time=0.3037, value=1}
            {arriveTangent=-3.1598, interpMode=RCIM_Cubic, leaveTangent=-3.1598, time=0.4429, value=0.444}
            <at 0.5885 0.1>
           ]}
         }}))
      Float=&Multiply_Float(
       A=&Lerp_Float(
        A=0.5
        B=1
        Alpha=&FloatFromCurve(
         FloatCurve=DataInterface() {propertyValues {Curve {keys [
             {arriveTangent=0.0028, interpMode=RCIM_Cubic, leaveTangent=0.0028, tangentMode=RCTM_User, time=0, value=1}
             {arriveTangent=-2.5249, interpMode=RCIM_Cubic, leaveTangent=-2.5249, tangentMode=RCTM_User, time=0.1408, value=0.9847}
             {arriveTangent=-0.0405, interpMode=RCIM_Cubic, leaveTangent=-0.0405, tangentMode=RCTM_User, time=0.8, value=0}
            ]}}}
         Scale Curve=&Lerp_Float(A=1, B=3, Alpha=$System.EffectIntensity)
        )
       )
       B=&FNTimeOfDayScalar(Night Scale=0.5, Morning Scale=0.7, Evening Scale=0.7)
      )
     )
     1=&MakeVectorFromLinearColorRGB(Color=&ColorFromCurve(ColorCurve=DataInterface() {propertyValues {
         AlphaCurve {keys []}
         BlueCurve {keys [<at 0.0764 2.5 RCIM_Cubic>,<at 0.3175 0 RCIM_Cubic>]}
         GreenCurve {keys [<at 0.0764 8.75 RCIM_Cubic>,<at 0.3175 0.1667 RCIM_Cubic>]}
         RedCurve {keys [<at 0.0764 40 RCIM_Cubic>,<at 0.3175 0.5 RCIM_Cubic>]}
        }}))
     2=?
     3=?
     Index=$Particles.VisibilityTag
     Num=::NewEnumerator0
    )
    ScaleA=true
    ScaleRGB=true
   ) {module=ScaleColor001}
  ]
  rendererClasses [&NiagaraSpriteRendererProperties, &NiagaraMeshRendererProperties]
  renderers [
   {rendererClass=&NiagaraSpriteRendererProperties}
   {rendererClass=&NiagaraMeshRendererProperties, rendererIndex=1}
  ]
 }
 FrontWind {
  emitterUpdateScript [
   &EmitterState(
    Inactive Response=::NewEnumerator1
    Life Cycle Mode=::NewEnumerator1
    Loop Behavior=::NewEnumerator1
    Loop Duration Mode=::NewEnumerator1
    UseLoopDelay=false
   )
   &SpawnBurst_Instantaneous(Use Spawn Probability=false, UseLoopCountLimit=true)
  ]
  particleSpawnScript [
   &InitializeParticle(
    Lifetime=0.5
    Mesh Scale=&Multiply_Vector(
     A=&VectorFromFloat(Value=&Multiply_Float(A=&RandomRangeFloat(Minimum=2.5, Maximum=2.5), B=1))
     B=<0.7 0.7 0.9>
    )
    Mesh Scale Mode=::NewEnumerator4
    Sprite Size Mode=::NewEnumerator3
    Uniform Sprite Size=999
    UsePositionOffset=true
    Write Material Random=true
   )
   &InitialMeshOrientation(
    Mesh Axis To Orient=<0 0 -1>
    Mesh Orientation Mode=::NewEnumerator2
    Orientation Vector=<0 1 0>
    Rotation=&FN_FrontEndVector3Modifier(Gameplay=<0.4 0 0>, Front End=<0.4 0 0>)
    Use Rotation Vector=true
   )
  ]
  particleUpdateScript [
   &ParticleState(Kill Particles When Lifetime Has Elapsed=false)
   &SocketLocation(
    SamplePosOffset=true
    Sampled Position Offset=&FN_FrontEndVector3Modifier(Gameplay=<0 0 0>, Front End=<0 0 -75>)
    Socket Reader=$System.Socket Reader
    Socket Sampling Mode=::NewEnumerator3
    VelocitySpeedLimit=false
   )
   &InterpolateOverTime(Rate Of Change=5, Target Quat Value:Quat4f=$Output.SocketLocation.SampledRotation, Variable Type=::NewEnumerator5)
   /SetVariables(Particles.MeshOrientation=&MultiplyQuaternion(Quaternion A=$StackContext.InterpolateOverTime.InterpolateOverTime_Quaternion.Moving Average, Quaternion B=$StackContext.InitialRotation)) {isSetParameters=true}
   &ScaleColor(
    Scale Alpha=&Multiply_Float(
     A=&Multiply_Float(A=0.03, B=&FNTimeOfDayScalar(Night Scale=1, Morning Scale=3, Evening Scale=1))
     B=&FN_FrontEndFloatModifier(
      Gameplay=&Multiply_Float(
       A=$System.Timeline.Time
       B=&FN_PickFloatDependOnPlatform(TargetPlatform=DataInterface() {propertyValues {Platforms {qualityLevelMask=-29}}}, TargetPlatformActive=0.05, TargetPlatformNoActive=1)
      )
      Front End=&Multiply_Float(
       A=&FloatFromCurve(
        FloatCurve=DataInterface() {propertyValues {Curve {
           defaultValue=1
           keys [{interpMode=RCIM_Cubic, tangentMode=RCTM_User, time=0, value=0}, {interpMode=RCIM_Cubic, tangentMode=RCTM_User, time=1, value=1}]
          }}}
        CurveIndex=$Emitter.Age
       )
       B=0.2
      )
     )
    )
    ScaleA=true
    ScaleRGB=false
   ) {module=ScaleColor001}
  ]
  rendererClasses [&NiagaraMeshRendererProperties]
  renderers [{rendererClass=&NiagaraMeshRendererProperties, rendererIndex=0}]
 }
 Sidewings {
  emitterSpawnScript [&FN_RibbonIDOffsetUtility()]
  emitterUpdateScript [
   &EmitterState(Inactive Response=::NewEnumerator1, Life Cycle Mode=::NewEnumerator1, Loop Delay=1, UseLoopDelay=true)
   &FN_SpawnAnimTrail(Number of Anim Trails=2)
  ]
  particleSpawnScript [
   &FN_InitializeAnimTrailBoneIndecies(First Bone or Socket Index=1, Second Bone or Socket Index=2)
   &FN_InitializeAnimTrailBoneIndecies(AnimTrail Selection=1, First Bone or Socket Index=3, Second Bone or Socket Index=4) {module=FN_InitializeAnimTrailBoneIndecies001}
   &FN_InitializeAnimTrail(Lifetime=0.2, Position Offset=&FN_FrontEndVector3Modifier(Gameplay=<0 0 0>, Front End=<0 0 0>))
   &AddVelocity(Velocity=<0 2500 0>)
  ]
  particleUpdateScript [
   &InheritSourceMovement(
    Applied Position Scale=&Multiply_VectorByFloat(
     Vector=<1 1 1>
     Float=&FloatFromCurve(FloatCurve=DataInterface() {propertyValues {Curve {keys [
          {interpMode=RCIM_Cubic, tangentMode=RCTM_User, time=0.66, value=1}
          {arriveTangent=-1.9042, interpMode=RCIM_Cubic, leaveTangent=-1.9042, tangentMode=RCTM_User, time=1, value=0.5}
         ]}}})
    )
    ApplyForce=false
    ApplyPosition=true
    ApplyPositionFalloff=false
   )
   &SolveForcesAndVelocity(Clamp Velocity=false, Limit Acceleration=false)
   &FN_UpdateAnimTrail() {module=FN_UpdateAnimTrail001}
   &DynamicMaterialParameters(
    Index 0 Param 1=&SelectFloatFromArray(Direct Array Index=$Particles.SpawnGroup, Float Selection Array=DataInterface() {propertyValues {FloatData [0,0.5]}})
    Index 1 Param 1=&FloatFromCurve(
     FloatCurve=DataInterface() {propertyValues {Curve {
        defaultValue=1
        keys [
         {arriveTangent=-2.9182, interpMode=RCIM_Cubic, leaveTangent=-2.9182, tangentMode=RCTM_User, time=0, value=1}
         {interpMode=RCIM_Cubic, tangentMode=RCTM_User, time=1, value=0}
        ]
       }}}
     Scale Curve=&Multiply_Float(
      A=&SelectFloatFromArray(Direct Array Index=$Particles.SpawnGroup, Float Selection Array=DataInterface() {propertyValues {FloatData [2,3]}})
      B=&FNTimeOfDayScalar(Night Scale=0.6, Morning Scale=1, Evening Scale=1)
     )
    )
    Module.Index0Float0Write=true
    Module.Index0Float1Write=true
    Module.Index0Float2Write=true
    Module.Index0Float3Write=true
    Module.Index1Float0Write=true
    Module.Index1Float1Write=true
    Module.Index1Float2Write=true
    Module.Index1Float3Write=true
    Write Parameter Index 1=::NewEnumerator1
   )
   &InterpolateOverTime(Initial Float Value:NiagaraFloat=$System.Timeline.Time, Target Float Value=1, Use Set Interpolation Rate=true)
   &KillParticles(Kill Particles=&SetBoolByFloatComparison(
     A=&FN_FrontEndFloatModifier(Gameplay=$System.Timeline.Time, Front End=1)
     B=$Particles.InterpolateOverTime.InterpolateOverTime_Float.Moving Average
     Comparison Type=::NewEnumerator6
    ))
  ]
  rendererClasses [&NiagaraRibbonRendererProperties]
  renderers [{rendererClass=&NiagaraRibbonRendererProperties}]
 }
 SpeedLine {
  emitterUpdateScript [
   &EmitterState()
   &SpawnRate(
    SpawnRate=&MakeCustomFloatFromBool(Bool=$User.bIsFrontEndPreview, True Float=50, False Float=90)
    Use Spawn Probability=true
   ) {module=SpawnRate001}
  ]
  particleSpawnScript [
   &InitializeParticle(
    Color=&MakeLinearColorFromVectorAndFloat(
     "Vector (RGB)"=<1 1 1>
     "Float (Alpha)"=&Multiply_Float(
      A=&UniformRangedFloat(Minimum=0.1, Maximum=0.18)
      B=&MakeCustomFloatFromBool(Bool=$User.bIsFrontEndPreview, True Float=0.5, False Float=0.75)
     )
    )
    Lifetime Max=0.16
    Lifetime Min=0.1
    Lifetime Mode=::NewEnumerator1
    Sprite Size Max=<3 10>
    Sprite Size Min=<1 2>
    Sprite Size Mode=::NewEnumerator2
    UsePositionOffset=false
    Write Material Random=true
   )
   &TorusLocation(
    Align Torus With Axis=true
    Handle Radius=0
    Large Radius:NiagaraFloat=&RandomRangeFloat(Minimum=150, Maximum=200)
    MaskSpawnGroup=false
    Module.Override Randomness=false
    Offset=<0 -180 0>
    Torus Axis=<0 1 0>
    Torus Origin:NiagaraPosition=&Add_Vector(
     A=Linked() {linkedVariable {name=Particles.Position, type=NiagaraPosition}}
     B=&MakeVector(X=0, Y=&MakeCustomFloatFromBool(Bool=$User.bIsFrontEndPreview, True Float=200, False Float=0), Z=0)
    )
   )
   &AddVelocityFromPoint(
    Module.Use Falloff Distance=false
    Origin Offset=<0 -600 0>
    Velocity Strength=&Multiply_Float(
     A=&FN_FrontEndFloatModifier(Gameplay=600, Front End=0)
     B=&MakeCustomFloatFromBool(Bool=$User.bIsFrontEndPreview, True Float=-1, False Float=1)
    )
   )
   &AddVelocity(Velocity=&Multiply_Vector(
     A=&RandomRangeVector(Minimum=<0 300 0>, Maximum=<0 700 0>)
     B=&MakeVector(X=1, Y=&MakeCustomFloatFromBool(Bool=$User.bIsFrontEndPreview, True Float=1, False Float=1), Z=1)
    ))
  ]
  particleUpdateScript [
   &SolveForcesAndVelocity(Clamp Velocity=false, Limit Acceleration=false)
   &ParticleState()
   &ScaleSpriteSize(Scale Factor=&Vector2DFromCurve(
     Vector2Curve=DataInterface() {propertyValues {XCurve {keys [<at 0 0.5 RCIM_Cubic>,<at 0.5 1 RCIM_Cubic>]}, YCurve {keys [<at 0 0 RCIM_Cubic>,<at 0.5 1 RCIM_Cubic>]}}}
     ScaleCurve=&Vector2DFromFloat(Value=1)
    ))
   &ScaleSpriteSizeBySpeed(Max Scale Factor=<1 50>, Min Scale Factor=<0.01 0.05>, Velocity Threshold=1)
   &ScaleColor(
    Scale Alpha=&FloatFromCurve(
     FloatCurve=DataInterface() {propertyValues {Curve {keys [
         <at 0 0 RCIM_Cubic>
         <at 0.3759 1.0207>
         {arriveTangent=-2.221, interpMode=RCIM_Cubic, leaveTangent=-2.221, tangentMode=RCTM_User, time=1, value=0}
        ]}}}
     Scale Curve=&MakeCustomFloatFromBool(
      Bool=$User.bIsFrontEndPreview
      True Float=&MakeCustomFloatFromBool(Bool=$User.bIsFullyDeployed, True Float=1, False Float=0)
      False Float=&Lerp_Float(A=0, B=1, Alpha=$User.ForwardDot)
     )
    )
    Scale RGB=&VectorFromFloat(Value=1)
    ScaleA=true
    ScaleRGB=true
   )
  ]
  rendererClasses [&NiagaraSpriteRendererProperties]
  renderers [{rendererClass=&NiagaraSpriteRendererProperties}]
 }
}
summary {
 systemName=NS_Glider_MorningSoak_Cook_Trail
 emitters [
  {emitterName=Sidewings, rendererClasses [&NiagaraRibbonRendererProperties]}
  {emitterName=FrontWind, rendererClasses [&NiagaraMeshRendererProperties]}
  {emitterName=SpeedLine, rendererClasses [&NiagaraSpriteRendererProperties]}
  {emitterName=Fire, rendererClasses [&NiagaraSpriteRendererProperties, &NiagaraMeshRendererProperties]}
  {
   emitterName=FakeLight_And_Fan
   rendererClasses [&NiagaraSpriteRendererProperties, &NiagaraMeshRendererProperties]
  }
 ]
 userVariables [
  User.AccentColor {type=NiagaraFloat, defaultValue=0, description=""}
  User.bIsFrontEnd {type=NiagaraBool}
  User.bIsFrontEndPreview {type=NiagaraBool}
  User.bIsFullyDeployed {type=NiagaraBool}
  User.ForwardDot {type=NiagaraFloat, defaultValue=0, description=""}
  User.RibbonIDOffset {type=NiagaraInt32}
  User.RightDot {type=NiagaraFloat, defaultValue=0, description=""}
  User.VariantID {type=NiagaraFloat, defaultValue=0, description=""}
 ]
}
